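; Gameplay ability for the laser underbarrel weapon mod, at present only a stub.
; Its failed-passive-activation event prints a debug string that refers to a mag passive.
; The activate-ability event has no logic attached, so the graph implements no laser behavior.
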(event Ability|EventFailedToActivatePassiveAbility
  (params (ActorInfo :type Gameplay%20Ability%20Actor%20Info%20Structure))
  (Development|PrintString "Failed to activate mag passive"))

(event Ability|EventActivateAbility)
